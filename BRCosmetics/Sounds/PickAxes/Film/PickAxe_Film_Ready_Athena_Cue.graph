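# Sound cue for the Film pickaxe's ready sound, built from that pickaxe's swing recordings.
# A random node picks among six equally weighted swing waves, feeding a modulator that raises pitch slightly and varies volume.
# It plays quietly in the melee weapon sound class with close projectile attenuation and a pickaxe-ready audio property sheet.
# The cue primes on load and overrides concurrency to block new instances rather than cut existing ones.
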SoundCue /BRCosmetics/Sounds/PickAxes/Film/PickAxe_Film_Ready_Athena_Cue
root=Modulator_0
nodes [
 {id=Modulator_0, children [Random_0], properties {pitchMax=1.2, pitchMin=1.1, volumeMax=1.1, volumeMin=1}}
 {
  id=Random_0
  children [WavePlayer_0, WavePlayer_5, WavePlayer_8, WavePlayer_9, WavePlayer_10, WavePlayer_11]
  properties {weights [1,1,1,1,1,1]}
 }
 {id=WavePlayer_0, properties {soundWaveAssetPtr=&/PA_Film_Swing_04}}
 {id=WavePlayer_10, properties {soundWaveAssetPtr=&/PA_Film_Swing_05}}
 {id=WavePlayer_11, properties {soundWaveAssetPtr=&/PA_Film_Swing_06}}
 {id=WavePlayer_5, properties {soundWaveAssetPtr=&/PA_Film_Swing_01}}
 {id=WavePlayer_8, properties {soundWaveAssetPtr=&/PA_Film_Swing_02}}
 {id=WavePlayer_9, properties {soundWaveAssetPtr=&/PA_Film_Swing_03}}
]
properties {
 attenuationSettings=&Fort_Weapon_Projectile_Close_Athena_Att
 audioPropertiesSheet=/BRCosmetics/Sounds/Settings/PropertySheets/Pickaxe/APS_Cosmetic_Pickaxe_Ready
 bOverrideConcurrency=true
 bPrimeOnLoad=true
 maxDistance=4950
 soundClassObject=&Fort_WeaponMelee
 volumeMultiplier=0.2
 concurrencyOverrides {resolutionRule=PreventNew}
}
